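# Unlit translucent Niagara sprite material for a flash spike in the PlasticGradient drum pickaxe effects.
# Blends the spike texture's red channel toward a gradient color texture, then tints it by particle color, fading from a bright boost to black over particle lifetime.
# A static switch can use particle color alone, before emissive adjustments.
# Opacity combines spike and particle alpha with depth fade; mobile skips that fade and the lifetime color ramp.
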Material /SparksCosmetics/Instrument/Drum/PlasticGradient/Effects/Pickaxe/M_Flash_Spike_VariantColor_PlasticGradient
graph {
 properties {bUsedWithNiagaraSprites=true, blendMode=BLEND_Translucent, pixelDepthOffsetMode=PDOM_Legacy, shadingModel=MSM_Unlit, usedShadingModels=Unlit}
 nodes [
  n1 = ParticleColor()
  n2 = TextureSample() {texture=/Game/Effects/Fort_Effects/Textures/Explo/T_Flash_Spike_01}
  n3 = TextureSampleParameter2D() {
   parameterName=TextureGradient
   texture=&/T_FX_Pickaxe_PlasticGradient_Drum_Color
  }
  n4 = LinearInterpolate(A=@n2.R, B=@n3.RGB, Alpha=ScalarParameter() {defaultValue=1, parameterName=Use Custom Texture}) {constAlpha=1}
  n5 = Multiply(A=LinearInterpolate(A=Multiply(B=@n1.RGB) {constA=8}, B=Multiply(A=@n1.RGB) {constB=0}, Alpha=ParticleRelativeTime()) {constA=6}, B=@n4)
  n6 = Multiply(A=@n1.RGB, B=@n4)
  MaterialFunctionCall(Emissive=ShadingPathSwitch(
    Default=StaticSwitchParameter(True=@n1.RGB, False=@n5) {parameterName=UseOnlyParticleColor}
    Mobile=StaticSwitchParameter(True=@n1.RGB, False=@n6) {parameterName=UseOnlyParticleColor}
   ) {desc=Migrated from FeatureLevelSwitch}) {materialFunction=&MF_FX_EmissiveAdjustments}
  n7 = Multiply(A=@n2.A, B=ParticleColor().A)
  n8 = DepthFade(Opacity=@n7, FadeDistance=ScalarParameter() {parameterName=DepthFade-Distance})
  ShadingPathSwitch(Default=@n8, Deferred=@n8, Mobile=@n7) {desc=Migrated from FeatureLevelSwitch}
 ]
}
